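# MF_Random2 is a material function that generates pseudo-random two-channel values from UVs, using the classic sine-hash technique.
# It takes the dot product of the UVs and a Seed Direction vector, then feeds it through sine for X and cosine for Y.
# Each result is scaled by Variation Mul and wrapped with Frac, then output separately and as a combined XY.
# All three inputs have defaults, with UVs falling back to texture coordinates.
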MaterialFunction /WildEstate/Assets/FX/General/MF_Random2
graph {nodes [
  n1 = FunctionInput() {bUsePreviewValueAsDefault=true, inputName=Seed Direction, inputType=FunctionInput_Vector2, sortPriority=1, previewValue=<12.9898 78.233 0 1>}
  n2 = DotProduct(A=FunctionInput(Preview=TextureCoordinate()) {bUsePreviewValueAsDefault=true, inputName=UVs, inputType=FunctionInput_Vector2}, B=@n1)
  n3 = FunctionInput() {bUsePreviewValueAsDefault=true, inputName=Variation Mul, inputType=FunctionInput_Scalar, sortPriority=2, previewValue=<43758.5469 0 0 1>}
  n4 = Frac(None=Multiply(A=Sine(None=@n2), B=@n3))
  n5 = Frac(None=Multiply(A=Cosine(None=@n2), B=@n3))
  FunctionOutput(None=AppendVector(A=FunctionOutput(None=@n4) {outputName=X, sortPriority=1}, B=FunctionOutput(None=@n5) {outputName=Y, sortPriority=2})) {outputName=XY}
 ]}
