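# Unlit Niagara material for sprites, ribbons and mesh particles that uses Alpha Holdout blending in the Before DOF translucency pass.
# A soft radial SphereMask centered in UV space is raised to the particle color's red channel, then multiplied by particle alpha, so particles control falloff and strength.
# A separate-translucency scene alpha sample, a constant and a DebugTimeSine call sit apart from that chain.
# No connections to material outputs are shown.
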Material /WildEstate/Assets/FX/General/M_OpacityComposite
graph {
 properties {
  bAutomaticallySetUsageInEditor=false
  bUsedWithNiagaraMeshParticles=true
  bUsedWithNiagaraRibbons=true
  bUsedWithNiagaraSprites=true
  blendMode=BLEND_AlphaHoldout
  shadingModel=MSM_Unlit
  translucencyPass=MTP_BeforeDOF
  usedShadingModels=Unlit
 }
 nodes [
  ComponentMask(None=SceneTexture(UVs=ScreenPosition().ViewportUV) {sceneTextureId=PPI_SeparateTranslucency}.Color) {a=true}
  Constant() {r=1}
  MaterialFunctionCall() {materialFunction=&DebugTimeSine}
  n1 = TextureCoordinate()
  n2 = Constant() {r=0.5}
  n3 = Constant()
  n4 = ParticleColor()
  Multiply(A=Power(Base=SphereMask(A=@n1, B=@n2, Radius=@n2, Hardness=@n3), Exp=@n4.R) {constExponent=3}, B=@n4.A) {constB=25}
 ]
}
