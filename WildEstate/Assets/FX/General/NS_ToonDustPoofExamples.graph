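# Example Niagara system showing three toon dust-poof variants side by side, plus a component-rendered emitter named Text.
# Each puff emitter bursts mesh particles with drag, curve-driven scale and time-of-day tint and shadow colors fed to dynamic material parameters.
# The low-frame variant uses NM_FPSTimer to gate NM_ToonOptions position and material jitter for a stepped look.
# The smooth variant disables both modules, while the middle one steps without position jitter.
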NiagaraSystem /WildEstate/Assets/FX/General/NS_ToonDustPoofExamples
dependencies {usedDataInterfaces [&NiagaraDataInterfaceCurve], usedDynamicInputs [], usedModules [&SystemState]}
emitters {
 NE_SmokePuffToonLowFrameLook {
  emitterUpdateScript [
   &EmitterState()
   &SpawnBurst_Instantaneous(Spawn Count=5, Use Spawn Probability=false, UseLoopCountLimit=false)
   &/NM_FPSTimer(Frames Per Second=6)
  ]
  particleSpawnScript [
   &InitializeParticle(
    Color=<rgba 0.5 0.5 0.5 0.75>
    Lifetime Max=5
    Lifetime Min=3
    Lifetime Mode=::NewEnumerator1
    Mesh Scale=<0.1 1.5 1.5>
    Mesh Scale Mode=::NewEnumerator4
    UsePositionOffset=false
    Write Material Random=true
   )
   &ShapeLocation(MaskSpawnGroup=false, Override Seed=false, UseOwnerScale=true)
   &AddVelocity(Module.Use Falloff Distance=false, Override Seed=false, Velocity Mode=::NewEnumerator1, Velocity Speed=150)
   &InitialMeshOrientation(
    Mesh Orientation Mode=::NewEnumerator7
    Rotation=&RandomRangeVector(Minimum=<0 0 0>, Maximum=<1 0 0>)
    Use Rotation Vector=true
   )
  ]
  particleUpdateScript [
   &ParticleState()
   &UpdateMeshOrientation(Rotation Rate=0)
   &ScaleMeshSize(Scale Factor=&VectorFromFloat(Value=&FloatFromCurve(FloatCurve=DataInterface() {propertyValues {Curve {keys [
          <at 0 0 RCIM_Cubic>
          {arriveTangent=1.3333, interpMode=RCIM_Cubic, leaveTangent=1.3333, time=0.1, value=1}
          {arriveTangent=-1, interpMode=RCIM_Cubic, leaveTangent=-1, time=0.75, value=1}
          <at 1 0.1 RCIM_Cubic>
         ]}}})))
   &Drag(Drag=2, Use Linear Drag=true, Use Rotational Drag=true)
   &SolveForcesAndVelocity(Clamp Velocity=false, Limit Acceleration=false)
   &Color(Color=&MultiplyLinearColor(
     Linear Color=&FNTimeOfDayColor(Night Color=<rgba 0.5 0.5 0.5 1>, Morning Color=<rgba 0.75 0.75 0.75 1>, Day Color=<rgba 1 1 1 1>, Evening Color=<rgba 0.75 0.75 0.75 1>)
     Scale Factor=&MakeVector4FromVectorAndFloat("Vector (RGB)"=&VectorFromFloat(Value=1), "Float (Alpha)"=0.75)
    ))
   /SetVariables(Particles.Shadow Color:LinearColor=&MultiplyLinearColor(
     Linear Color=&FNTimeOfDayColor(Night Color=<rgba 0.35 0.35 0.35 1>, Morning Color=<rgba 0.4 0.4 0.4 1>, Day Color=<rgba 0.5 0.5 0.5 1>, Evening Color=<rgba 0.4 0.4 0.4 1>)
     Scale Factor=&MakeVector4FromVectorAndFloat("Vector (RGB)"=&VectorFromFloat(Value=1), "Float (Alpha)"=0.75)
    )) {isSetParameters=true}
   &DynamicMaterialParameters(
    Index 0 Vector 1:Vector3f=&MakeVectorFromLinearColorRGB(Color=$Particles.Shadow Color)
    Index 1 Param 1=1
    Index 1 Param 2=&FloatFromCurve(FloatCurve=DataInterface() {propertyValues {Curve {keys [<at 0.5 1 RCIM_Cubic>,<at 1 0>]}}}, Scale Curve=2)
    Module.Index0Float3Write=true
    Module.Index0VectorWrite=true
    Module.Index1Float0Write=true
    Module.Index1Float1Write=true
    Module.Index1Float2Write=true
    Module.Index1Float3Write=true
    Write Parameter Index 0=::NewEnumerator3
    Write Parameter Index 1=::NewEnumerator1
   )
   &/NM_ToonOptions(
    Dynamic1Jitter:Vector4f=&MakeVector4(
     X=&MakeFloatFromInt(Int=&MakeIntFromFloat(Float=&Lerp_Float(
        A=10
        B=4
        Alpha=&FloatFromCurve(FloatCurve=DataInterface() {propertyValues {Curve {defaultValue=1, keys [{tangentMode=RCTM_User, time=0, value=0}, {tangentMode=RCTM_User, time=1, value=1}]}}})
       )))
     Y=0
     Z=0
     W=0
    )
    DynamicJitter:Vector4f=&MakeVector4(
     X=0
     Y=0
     Z=0
     W=&Add_Float(
      A=&Add_Float(
       A=&Multiply_Float(A=$Engine.Time, B=0.1)
       B=&RandomRangeFloat(Minimum=-0.0035, Maximum=0.0035, Evaluation Type=::NewEnumerator1)
      )
      B=&RandomRangeFloat(Minimum=0, Maximum=1000)
     )
    )
    JitterPos:Vector3f=&RandomVector(
     Vector Scale=&RandomRangeFloat(Minimum=1, Maximum=5, Evaluation Type=::NewEnumerator1)
     Evaluation Type=::NewEnumerator1
    )
    UpdateThisFrame:NiagaraBool=$Emitter.NM_FPSTimer.Timer Active
   )
   /SetVariables(Particles.Test_CamOffset=10) {isSetParameters=true}
  ]
  rendererClasses [&NiagaraMeshRendererProperties]
  renderers [
   {rendererClass=&NiagaraMeshRendererProperties, rendererIndex=0}
   {rendererClass=&NiagaraMeshRendererProperties, rendererIndex=1}
  ]
 }
 NE_SmokePuffToonSmooveLook {
  emitterUpdateScript [
   &EmitterState()
   &SpawnBurst_Instantaneous(Spawn Count=5, Use Spawn Probability=false, UseLoopCountLimit=false)
   &NM_FPSTimer(Frames Per Second=6) {enabled=false}
  ]
  particleSpawnScript [
   &InitializeParticle(
    Color=<rgba 0.5 0.5 0.5 0.75>
    Lifetime Max=5
    Lifetime Min=3
    Lifetime Mode=::NewEnumerator1
    Mesh Scale=<0.1 1.5 1.5>
    Mesh Scale Mode=::NewEnumerator4
    UsePositionOffset=false
    Write Material Random=true
   )
   &AddVelocity(Module.Use Falloff Distance=false, Override Seed=false, Velocity Mode=::NewEnumerator1, Velocity Speed=150)
   &ShapeLocation(MaskSpawnGroup=false, Offset=<1000 0 0>, Offset Mode=::NewEnumerator0, Override Seed=false, UseOwnerScale=true)
   &InitialMeshOrientation(
    Mesh Orientation Mode=::NewEnumerator7
    Rotation=&RandomRangeVector(Minimum=<0 0 0>, Maximum=<1 0 0>)
    Use Rotation Vector=true
   )
  ]
  particleUpdateScript [
   &ParticleState()
   &ScaleMeshSize(Scale Factor=&VectorFromFloat(Value=&FloatFromCurve(FloatCurve=DataInterface() {propertyValues {Curve {
         defaultValue=1
         keys [
          <at 0 0 RCIM_Cubic>
          {arriveTangent=1.3333, interpMode=RCIM_Cubic, leaveTangent=1.3333, time=0.1, value=1}
          {arriveTangent=-1, interpMode=RCIM_Cubic, leaveTangent=-1, time=0.75, value=1}
          <at 1 0.1 RCIM_Cubic>
         ]
        }}})))
   &Drag(Drag=2, Use Linear Drag=true, Use Rotational Drag=true)
   &SolveForcesAndVelocity(Clamp Velocity=false, Limit Acceleration=false)
   &Color(Color=&MultiplyLinearColor(
     Linear Color=&FNTimeOfDayColor(Night Color=<rgba 0.5 0.5 0.5 1>, Morning Color=<rgba 0.75 0.75 0.75 1>, Day Color=<rgba 1 1 1 1>, Evening Color=<rgba 0.75 0.75 0.75 1>)
     Scale Factor=&MakeVector4FromVectorAndFloat("Vector (RGB)"=&VectorFromFloat(Value=1), "Float (Alpha)"=0.75)
    ))
   /SetVariables(Particles.Shadow Color:LinearColor=&MultiplyLinearColor(
     Linear Color=&FNTimeOfDayColor(Night Color=<rgba 0.35 0.35 0.35 1>, Morning Color=<rgba 0.4 0.4 0.4 1>, Day Color=<rgba 0.5 0.5 0.5 1>, Evening Color=<rgba 0.4 0.4 0.4 1>)
     Scale Factor=&MakeVector4FromVectorAndFloat("Vector (RGB)"=&VectorFromFloat(Value=1), "Float (Alpha)"=0.75)
    )) {isSetParameters=true, module=SetVariables_#1}
   &DynamicMaterialParameters(
    Index 0 Param 4=&Add_Float(
     A=&Multiply_Float(A=$Engine.Time, B=0.05)
     B=&RandomRangeFloat(Minimum=0, Maximum=1000)
    )
    Index 0 Vector 1:Vector3f=&MakeVectorFromLinearColorRGB(Color=$Particles.Shadow Color)
    Index 1 Param 1=&MakeFloatFromInt(Int=&RandomRangeInt(Minimum=7, Maximum=5))
    Index 1 Param 2=&FloatFromCurve(FloatCurve=DataInterface() {propertyValues {Curve {keys [<at 0.5 1 RCIM_Cubic>,<at 1 0>]}}}, Scale Curve=2)
    Module.Index0Float3Write=true
    Module.Index0VectorWrite=true
    Module.Index1Float0Write=true
    Module.Index1Float1Write=true
    Module.Index1Float2Write=true
    Module.Index1Float3Write=true
    Write Parameter Index 0=::NewEnumerator3
    Write Parameter Index 1=::NewEnumerator1
   )
   &NM_ToonOptions(
    Dynamic1Jitter:Vector4f=&MakeVector4(
     X=&MakeFloatFromInt(Int=&MakeIntFromFloat(Float=&Lerp_Float(
        A=10
        B=4
        Alpha=&FloatFromCurve(FloatCurve=DataInterface() {propertyValues {Curve {defaultValue=1, keys [{tangentMode=RCTM_User, time=0, value=0}, {tangentMode=RCTM_User, time=1, value=1}]}}})
       )))
     Y=0
     Z=0
     W=0
    )
    DynamicJitter:Vector4f=&MakeVector4(
     X=0
     Y=0
     Z=0
     W=&Add_Float(
      A=&Add_Float(
       A=&Multiply_Float(A=$Engine.Time, B=0.1)
       B=&RandomRangeFloat(Minimum=-0.0035, Maximum=0.0035, Evaluation Type=::NewEnumerator1)
      )
      B=&RandomRangeFloat(Minimum=0, Maximum=1000)
     )
    )
    JitterPos:Vector3f=&RandomVector(
     Vector Scale=&RandomRangeFloat(Minimum=1, Maximum=5, Evaluation Type=::NewEnumerator1)
     Evaluation Type=::NewEnumerator1
    )
    RotationJitter:Quat4f=&MakeQuaternion(
     Quaternion From=1
     Axis=<1 0 0>
     Angle=0
     Roll=&Add_Float(
      A=&RandomRangeFloat(Minimum=0, Maximum=360)
      B=&RandomRangeFloat(Minimum=-5, Maximum=5, Evaluation Type=::NewEnumerator1)
     )
     From=?
     To=?
     Angle Type=::NewEnumerator0
     Normalize Vectors=false
    )
    UpdateThisFrame:NiagaraBool=$Emitter.NM_FPSTimer.Timer Active
   ) {enabled=false}
   /SetVariables(Particles.Test_CamOffset=10) {isSetParameters=true}
  ]
  rendererClasses [&NiagaraMeshRendererProperties]
  renderers [
   {rendererClass=&NiagaraMeshRendererProperties, rendererIndex=0}
   {rendererClass=&NiagaraMeshRendererProperties, rendererIndex=1}
  ]
 }
 NE_SmokePuffToonSmooveLook001 {
  emitterUpdateScript [
   &EmitterState()
   &SpawnBurst_Instantaneous(Spawn Count=5, Use Spawn Probability=false, UseLoopCountLimit=false)
   &NM_FPSTimer(Frames Per Second=6)
  ]
  particleSpawnScript [
   &InitializeParticle(
    Color=<rgba 0.5 0.5 0.5 0.75>
    Lifetime Max=5
    Lifetime Min=3
    Lifetime Mode=::NewEnumerator1
    Mesh Scale=<0.1 1.5 1.5>
    Mesh Scale Mode=::NewEnumerator4
    UsePositionOffset=false
    Write Material Random=true
   )
   &AddVelocity(Module.Use Falloff Distance=false, Override Seed=false, Velocity Mode=::NewEnumerator1, Velocity Speed=150)
   &ShapeLocation(MaskSpawnGroup=false, Offset=<500 0 0>, Offset Mode=::NewEnumerator0, Override Seed=false, UseOwnerScale=true)
   &InitialMeshOrientation(
    Mesh Orientation Mode=::NewEnumerator7
    Rotation=&RandomRangeVector(Minimum=<0 0 0>, Maximum=<1 0 0>)
    Use Rotation Vector=true
   )
  ]
  particleUpdateScript [
   &ParticleState()
   &ScaleMeshSize(Scale Factor=&VectorFromFloat(Value=&FloatFromCurve(FloatCurve=DataInterface() {propertyValues {Curve {
         defaultValue=1
         keys [
          <at 0 0 RCIM_Cubic>
          {arriveTangent=1.3333, interpMode=RCIM_Cubic, leaveTangent=1.3333, time=0.1, value=1}
          {arriveTangent=-1, interpMode=RCIM_Cubic, leaveTangent=-1, time=0.75, value=1}
          <at 1 0.1 RCIM_Cubic>
         ]
        }}})))
   &Drag(Drag=2, Use Linear Drag=true, Use Rotational Drag=true)
   &SolveForcesAndVelocity(Clamp Velocity=false, Limit Acceleration=false)
   &Color(Color=&MultiplyLinearColor(
     Linear Color=&FNTimeOfDayColor(Night Color=<rgba 0.5 0.5 0.5 1>, Morning Color=<rgba 0.75 0.75 0.75 1>, Day Color=<rgba 1 1 1 1>, Evening Color=<rgba 0.75 0.75 0.75 1>)
     Scale Factor=&MakeVector4FromVectorAndFloat("Vector (RGB)"=&VectorFromFloat(Value=1), "Float (Alpha)"=0.75)
    ))
   /SetVariables(Particles.Shadow Color:LinearColor=&MultiplyLinearColor(
     Linear Color=&FNTimeOfDayColor(Night Color=<rgba 0.35 0.35 0.35 1>, Morning Color=<rgba 0.4 0.4 0.4 1>, Day Color=<rgba 0.5 0.5 0.5 1>, Evening Color=<rgba 0.4 0.4 0.4 1>)
     Scale Factor=&MakeVector4FromVectorAndFloat("Vector (RGB)"=&VectorFromFloat(Value=1), "Float (Alpha)"=0.75)
    )) {isSetParameters=true, module=SetVariables_#1}
   &DynamicMaterialParameters(
    Index 0 Param 4=&Add_Float(
     A=&Multiply_Float(A=$Engine.Time, B=0)
     B=&RandomRangeFloat(Minimum=0, Maximum=1000)
    )
    Index 0 Vector 1:Vector3f=&MakeVectorFromLinearColorRGB(Color=$Particles.Shadow Color)
    Index 1 Param 1=1
    Index 1 Param 2=&FloatFromCurve(FloatCurve=DataInterface() {propertyValues {Curve {keys [<at 0.5 1 RCIM_Cubic>,<at 1 0>]}}}, Scale Curve=2)
    Module.Index0Float3Write=true
    Module.Index0VectorWrite=true
    Module.Index1Float0Write=true
    Module.Index1Float1Write=true
    Module.Index1Float2Write=true
    Module.Index1Float3Write=true
    Write Parameter Index 0=::NewEnumerator3
    Write Parameter Index 1=::NewEnumerator1
   )
   &NM_ToonOptions(
    Dynamic1Jitter:Vector4f=&MakeVector4(
     X=&MakeFloatFromInt(Int=&MakeIntFromFloat(Float=&Lerp_Float(
        A=10
        B=4
        Alpha=&FloatFromCurve(FloatCurve=DataInterface() {propertyValues {Curve {defaultValue=1, keys [{tangentMode=RCTM_User, time=0, value=0}, {tangentMode=RCTM_User, time=1, value=1}]}}})
       )))
     Y=0
     Z=0
     W=0
    )
    DynamicJitter:Vector4f=&MakeVector4(
     X=0
     Y=0
     Z=0
     W=&Add_Float(
      A=&Add_Float(
       A=&Multiply_Float(A=$Engine.Time, B=0.05)
       B=&RandomRangeFloat(Minimum=-0.0035, Maximum=0.0035, Evaluation Type=::NewEnumerator1)
      )
      B=&RandomRangeFloat(Minimum=0, Maximum=1000)
     )
    )
    UpdateThisFrame:NiagaraBool=$Emitter.NM_FPSTimer.Timer Active
   )
   /SetVariables(Particles.Test_CamOffset=10) {isSetParameters=true}
  ]
  rendererClasses [&NiagaraMeshRendererProperties]
  renderers [
   {rendererClass=&NiagaraMeshRendererProperties, rendererIndex=0}
   {rendererClass=&NiagaraMeshRendererProperties, rendererIndex=1}
  ]
 }
 Text {
  emitterUpdateScript [
   &EmitterState()
   &SpawnBurst_Instantaneous(Use Spawn Probability=false, UseLoopCountLimit=false)
  ]
  particleSpawnScript [
   &InitializeParticle(Color Mode=::NewEnumerator0, Lifetime=5, Position Offset=<-200 0 250>, UsePositionOffset=true, Write Material Random=true)
   &InitialMeshOrientation(Mesh Orientation Mode=::NewEnumerator7, Rotation=<0 0 0.25>, Use Rotation Vector=true)
  ]
  particleUpdateScript [&ParticleState()]
  rendererClasses [&NiagaraComponentRendererProperties]
  renderers [{rendererClass=&NiagaraComponentRendererProperties, rendererIndex=0}]
 }
}
summary {
 systemName=NS_ToonDustPoofExamples
 emitters [
  {emitterName=NE_SmokePuffToonLowFrameLook, rendererClasses [&NiagaraMeshRendererProperties]}
  {emitterName=NE_SmokePuffToonSmooveLook, rendererClasses [&NiagaraMeshRendererProperties]}
  {emitterName=NE_SmokePuffToonSmooveLook001, rendererClasses [&NiagaraMeshRendererProperties]}
  {emitterName=Text, rendererClasses [&NiagaraComponentRendererProperties]}
 ]
 userVariables []
}
